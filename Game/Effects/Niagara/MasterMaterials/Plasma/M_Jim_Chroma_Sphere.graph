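# Unlit translucent Niagara master material for a chromatic plasma sphere effect on mesh and sprite particles.
# Three offset samples of panning, normal-warped grayscale noise form RGB channels, or blend to a color texture, with saturation adjustment and particle tint.
# Dynamic parameters drive warp, UV offset and contrast.
# Fresnel, gradient and optional hemisphere masks shape opacity, with depth and near-camera fades skipped on mobile; Fresnel also drives refraction.
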Material /Game/Effects/Niagara/MasterMaterials/Plasma/M_Jim_Chroma_Sphere
graph {
 properties {
  bUsedWithMeshParticles=true
  bUsedWithNiagaraMeshParticles=true
  bUsedWithParticleSprites=true
  blendMode=BLEND_Translucent
  pixelDepthOffsetMode=PDOM_Legacy
  refractionMethod=RM_IndexOfRefraction
  shadingModel=MSM_Unlit
  usedShadingModels=Unlit
 }
 nodes [
  n1 = OneMinus(None=Fresnel() {baseReflectFraction=0, exponent=1})
  n2 = Multiply(A=@n1, B=@n1)
  n3 = Multiply(
   A=Multiply(A=@n2, B=@n2)
   B=Fresnel(BaseReflectFractionIn=ScalarParameter() {group=Refraction, parameterName=CenterOpacity_Refrac, sortPriority=2}) {baseReflectFraction=0, exponent=1}
  )
  n4 = ParticleColor()
  LinearInterpolate(
   B=LinearInterpolate(B=ScalarParameter() {defaultValue=1.1, group=Refraction, parameterName=RefractionStr, sortPriority=1}, Alpha=Multiply(A=@n3, B=@n4.A)) {constA=1, constB=1.1}
   Alpha=ScalarParameter() {defaultValue=1, group=Refraction, parameterName=Use Refraction}
  ) {constA=1}
  n5 = DynamicParameter() {paramNames [UVWarpStr,RandX,RandY,Power]}
  n6 = AppendVector(A=@n5.RandX, B=@n5.RandX)
  n7 = ScalarParameter() {defaultValue=2, group=Chromatic Texture, parameterName=BaseTilingU, sortPriority=6}
  n8 = ScalarParameter() {defaultValue=1, group=Chromatic Texture, parameterName=BaseTilingV, sortPriority=7}
  n9 = TextureCoordinate()
  n10 = AppendVector(
   A=ScalarParameter() {defaultValue=2, group=Distortion, parameterName=WarpTexTilingU, sortPriority=3}
   B=ScalarParameter() {defaultValue=1, group=Distortion, parameterName=WarpTexTilingV, sortPriority=4}
  )
  n11 = ScalarParameter() {group=Distortion, parameterName=WarpPanSpeedU, sortPriority=5}
  n12 = ScalarParameter() {defaultValue=0.25, group=Distortion, parameterName=WarpPanSpeedV, sortPriority=6}
  n13 = TextureSampleParameter2D(UVs=Panner(Coordinate=Multiply(A=@n9, B=@n10), Speed=AppendVector(A=@n11, B=@n12)) {bFractionalPart=true, speedY=0.25}) {
   group=Distortion
   parameterName=UVDistTex
   samplerType=SAMPLERTYPE_Normal
   sortPriority=2
   texture=/Game/Global/Textures/Default/Noise/T_SwirlNoise_N
  }
  n14 = ScalarParameter() {defaultValue=0.07, group=Distortion, parameterName=UVWarpStr, sortPriority=1}
  n15 = DynamicParameter() {paramNames [UVWarpStr,RandX,RandY,Power]}
  n16 = ScalarParameter() {defaultValue=1, group=Distortion, parameterName=Use DynParam_DistortUV}
  n17 = Add(
   A=Multiply(A=TextureCoordinate(), B=AppendVector(A=@n7, B=@n8))
   B=Multiply(A=ComponentMask(None=@n13.RGB) {g=true, r=true}, B=LinearInterpolate(A=@n14, B=@n15.UVWarpStr, Alpha=@n16)) {constB=0.0125}
  )
  n18 = Add(
   A=@n17
   B=Multiply(
    A=Constant2Vector() {r=1}
    B=ScalarParameter() {defaultValue=0.025, group=Chromatic Texture, parameterName=ChromaShiftDistance, sortPriority=10}
   )
  )
  n19 = ScalarParameter() {defaultValue=0.25, group=Chromatic Texture, parameterName=PanSpeedU, sortPriority=8}
  n20 = ScalarParameter() {defaultValue=0.5, group=Chromatic Texture, parameterName=PanSpeedV, sortPriority=9}
  n21 = TextureSampleParameter2D(UVs=Panner(Coordinate=Add(A=@n6, B=@n18), Speed=AppendVector(A=@n19, B=@n20)) {bFractionalPart=true, speedY=0.25}) {group=Chromatic Texture, parameterName=GrayscaleTex, sortPriority=1, texture=/Game/Global/Textures/Default/Noise/T_FireNoise_02}
  n22 = ChannelMaskParameter(None=@n21.RGBA) {group=Chromatic Texture, maskChannel=Green, parameterName=GreyScale_X, sortPriority=2, defaultValue=<rgba 0 1 0 0>}
  n23 = DynamicParameter() {paramNames [UVWarpStr,RandX,RandY,Power]}
  n24 = DynamicParameter() {paramNames [UVWarpStr,RandX,RandY,Power]}
  n25 = AppendVector(A=@n24.RandX, B=@n24.RandX)
  n26 = ScalarParameter() {defaultValue=0.25, group=Chromatic Texture, parameterName=PanSpeedU, sortPriority=8}
  n27 = ScalarParameter() {defaultValue=0.5, group=Chromatic Texture, parameterName=PanSpeedV, sortPriority=9}
  n28 = TextureSampleParameter2D(UVs=Panner(Coordinate=Add(A=@n25, B=@n17), Speed=AppendVector(A=@n26, B=@n27)) {bFractionalPart=true, speedY=0.25}) {group=Chromatic Texture, parameterName=GrayscaleTex, sortPriority=1, texture=&T_FireNoise_02}
  n29 = ChannelMaskParameter(None=@n28.RGBA) {group=Chromatic Texture, maskChannel=Green, parameterName=GreyScale_Y, sortPriority=3, defaultValue=<rgba 0 1 0 0>}
  n30 = DynamicParameter() {paramNames [UVWarpStr,RandX,RandY,Power]}
  n31 = DynamicParameter() {paramNames [UVWarpStr,RandX,RandY,Power]}
  n32 = AppendVector(A=@n31.RandX, B=@n31.RandX)
  n33 = Add(
   A=@n17
   B=Multiply(
    A=ScalarParameter() {defaultValue=0.025, group=Chromatic Texture, parameterName=ChromaShiftDistance, sortPriority=10}
    B=Constant2Vector() {r=-1}
   )
  )
  n34 = ScalarParameter() {defaultValue=0.25, group=Chromatic Texture, parameterName=PanSpeedU, sortPriority=8}
  n35 = ScalarParameter() {defaultValue=0.5, group=Chromatic Texture, parameterName=PanSpeedV, sortPriority=9}
  n36 = TextureSampleParameter2D(UVs=Panner(Coordinate=Add(A=@n32, B=@n33), Speed=AppendVector(A=@n34, B=@n35)) {bFractionalPart=true, speedY=0.25}) {group=Chromatic Texture, parameterName=GrayscaleTex, sortPriority=1, texture=&T_FireNoise_02}
  n37 = ChannelMaskParameter(None=@n36.RGBA) {group=Chromatic Texture, maskChannel=Green, parameterName=GreyScale_Z, sortPriority=4, defaultValue=<rgba 0 1 0 0>}
  n38 = DynamicParameter() {paramNames [UVWarpStr,RandX,RandY,Power]}
  n39 = DynamicParameter() {paramNames [UVWarpStr,RandX,RandY,Power]}
  n40 = AppendVector(A=@n39.RandX, B=@n39.RandX)
  n41 = ScalarParameter() {defaultValue=0.25, parameterName=PanSpeedU}
  n42 = ScalarParameter() {defaultValue=0.5, parameterName=PanSpeedV}
  n43 = TextureSampleParameter2D(UVs=Panner(Coordinate=Add(A=@n40, B=@n17), Speed=AppendVector(A=@n41, B=@n42)) {bFractionalPart=true, speedY=0.25}) {
   group=Chromatic Texture
   parameterName=ColorTex
   sortPriority=5
   texture=&T_Inky_Smoke_Tile_Chromatic
  }
  n44 = LinearInterpolate(
   A=MaterialFunctionCall(
    X=Power(Base=@n22.R, Exp=@n23.Power) {constExponent=14}
    Y=Power(Base=@n29.R, Exp=@n30.Power) {constExponent=14}
    Z=Power(Base=@n37.R, Exp=@n38.Power) {constExponent=14}
   ) {materialFunction=&MakeFloat3}.Result
   B=Power(Base=@n43.RGB, Exp=DynamicParameter() {paramNames [UVWarpStr,RandX,RandY,Power]}.Power) {constExponent=14}
   Alpha=ScalarParameter() {group=Chromatic Texture, parameterName=Use Chromatic Color Texture}
  )
  n45 = Desaturation(None=@n44, Fraction=ScalarParameter() {defaultValue=-3, group=Chromatic Texture, parameterName=SaturationFactor, sortPriority=11})
  n46 = ParticleColor()
  MaterialFunctionCall(Emissive=MaterialFunctionCall(LDR Input=Clamp(None=@n45), HDR Tint=Multiply(A=@n45, B=@n46.RGB)) {materialFunction=&DeriveHDRfromLDR}.RemappedRange) {materialFunction=&MF_FX_EmissiveAdjustments}
  n47 = ScalarParameter() {defaultValue=600, group=Opacity, parameterName=DepthFadeDistance, sortPriority=2}
  n48 = TextureSampleParameter2D(UVs=MaterialFunctionCall(XY=TextureCoordinate()) {materialFunction=&Swizzle}.YX) {parameterName=Mask_Texture, samplerType=SAMPLERTYPE_LinearColor, texture=/Game/Global/Textures/Default/Masks_and_Gradients/T_VertGradient}
  n49 = OneMinus(None=Fresnel() {baseReflectFraction=0, exponent=1})
  n50 = Multiply(A=@n49, B=@n49)
  n51 = ScalarParameter() {group=Opacity, parameterName=CenterOpacity}
  n52 = Multiply(A=@n48.RGB, B=Multiply(A=Multiply(A=@n50, B=@n50), B=Fresnel(BaseReflectFractionIn=@n51) {baseReflectFraction=0, exponent=1}))
  n53 = ComponentMask(None=TextureCoordinate()) {g=true}
  n54 = ScalarParameter() {group=Opacity, parameterName=Hemisphere Edge Contrast}
  n55 = Saturate(None=Multiply(A=Power(Base=@n53, Exp=@n54), B=ScalarParameter() {defaultValue=1, group=Opacity, parameterName=Hemisphere Intensity}))
  n56 = ScalarParameter() {group=Opacity, parameterName=MakeHemiSphere}
  n57 = Multiply(A=@n46.A, B=Multiply(A=Multiply(A=@n44, B=@n52), B=LinearInterpolate(B=@n55, Alpha=@n56) {constA=1}))
  n58 = ScalarParameter() {defaultValue=25, group=Opacity, parameterName=NearCamFade, sortPriority=1}
  ShadingPathSwitch(
   Default=Multiply(
    A=DepthFade(Opacity=@n57, FadeDistance=@n47)
    B=MaterialFunctionCall(Distance=@n58) {materialFunction=&Near_Camera_Fade}.Result
   )
   Mobile=@n57
  )
 ]
}
