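# Translucent, unlit, two-sided Niagara master material for energy sphere effects.
# Panning noise distorts scrolling base and detail textures, which sum into HDR emissive tinted by particle color.
# Fresnel, view-angle fade and optional linear or texture masks shape opacity, with depth fade on deferred.
# Dynamic parameters scale V-masked noise displacement along normals and optional caustic-textured rim refraction, which forward and mobile paths skip.
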Material /Game/Effects/Niagara/MasterMaterials/Plasma/M_Sphere_Energy
graph {
 properties {
  bAutomaticallySetUsageInEditor=false
  bUsedWithNiagaraMeshParticles=true
  bUsedWithNiagaraRibbons=true
  bUsedWithNiagaraSprites=true
  blendMode=BLEND_Translucent
  refractionMethod=RM_IndexOfRefraction
  shadingModel=MSM_Unlit
  twoSided=true
  usedShadingModels=Unlit
 }
 nodes [
  n1 = Multiply(
   A=TextureCoordinate()
   B=AppendVector(
    A=ScalarParameter() {defaultValue=4, group=Base, parameterName=BaseTexture_U}
    B=ScalarParameter() {defaultValue=4, group=Base, parameterName=BaseTexture_V}
   )
  )
  n2 = TextureCoordinate()
  n3 = ScalarParameter() {defaultValue=10, group=Noise, parameterName=NoiseTexture_Tile}
  n4 = ScalarParameter() {defaultValue=2, group=Noise, parameterName=NoiseTexture_SpeedX}
  n5 = ScalarParameter() {defaultValue=2, group=Noise, parameterName=NoiseTexture_SpeedY}
  n6 = TextureSampleParameter2D(UVs=Panner(Coordinate=Multiply(A=@n2, B=@n3) {constB=10}, Speed=AppendVector(A=@n4, B=@n5)) {speedX=2, speedY=2}) {group=Noise, parameterName=NoiseTexture, texture=/Game/Effects/Fort_Effects/Emotes/UltraLight/Texture/T_SoftNoise_BC7}
  n7 = Multiply(A=ComponentMask(None=@n6.RGB) {g=true, r=true}, B=ScalarParameter() {defaultValue=0.15, group=Base, parameterName=BaseTexture_Noise}) {constB=0.15}
  n8 = ScalarParameter() {group=Base, parameterName=BaseTextureSpeed_U}
  n9 = ScalarParameter() {group=Base, parameterName=BaseTextureSpeed_V}
  n10 = TextureSampleParameter2D(UVs=Panner(Coordinate=Add(A=@n1, B=@n7), Speed=AppendVector(A=@n8, B=@n9)) {speedY=0.2}) {group=Base, parameterName=BaseTexture, texture=/Game/Characters/Player/Male/Medium/Bodies/M_Med_Soldier_04/Skins/Galaxy/Textures/T_FN_Nebula}
  n11 = TextureCoordinate()
  n12 = AppendVector(
   A=ScalarParameter() {defaultValue=4, group=Detail, parameterName=DetailTexture_U}
   B=ScalarParameter() {defaultValue=4, group=Detail, parameterName=DetailTexture_V}
  )
  n13 = ComponentMask(None=@n6.RGB) {b=true, g=true}
  n14 = ScalarParameter() {defaultValue=0.15, group=Detail, parameterName=DetailTexture_Noise}
  n15 = Panner(
   Coordinate=Add(A=Multiply(A=@n11, B=@n12), B=Multiply(A=@n13, B=@n14) {constB=0.15})
   Speed=AppendVector(
    A=ScalarParameter() {group=Detail, parameterName=DetailTextureSpeed_U}
    B=ScalarParameter() {group=Detail, parameterName=DetailTextureSpeed_V}
   )
  ) {speedY=0.2}
  n16 = Add(
   A=Desaturation(None=@n10.RGB, Fraction=ScalarParameter() {defaultValue=1, parameterName=Saturation})
   B=Multiply(A=TextureSampleParameter2D(UVs=@n15) {group=Detail, parameterName=DetailTexture, texture=/Game/Effects/Fort_Effects/Textures/Tile/T_Noise_02}.RGB) {constB=1.5}
  )
  n17 = MaterialFunctionCall(LDR Input=@n16, HDR Tint=@n16, HDR Intensity Multiplier=ScalarParameter() {defaultValue=4, parameterName=HDR_Light}) {materialFunction=&DeriveHDRfromLDR}
  n18 = ParticleColor()
  MaterialFunctionCall(Emissive=Multiply(A=Multiply(A=@n17.RemappedRange, B=@n18.RGB), B=ScalarParameter() {defaultValue=1, parameterName=All light})) {materialFunction=&MF_FX_EmissiveAdjustments}
  n19 = TextureCoordinate()
  n20 = ScalarParameter() {defaultValue=4, parameterName=RefractionTex_Tile}
  n21 = TextureSampleParameter2D(UVs=Panner(Coordinate=Multiply(A=@n19, B=@n20) {constB=4}) {speedX=0.5, speedY=0.5}) {
   parameterName=RefractionTex
   samplerType=SAMPLERTYPE_LinearGrayscale
   texture=/Game/Global/Textures/Default/Caustics/T_Nosie_Caustics_Grayscale
  }
  n22 = LinearInterpolate(Alpha=@n21.R)
  Multiply(A=@n22)
  n23 = TextureCoordinate()
  n24 = Sine(None=Multiply(
    A=MaterialFunctionCall(Float2=@n23) {materialFunction=&BreakOutFloat2Components}.G
    B=ScalarParameter() {defaultValue=0.8, parameterName=WPOMaskV_Multiply}
   ) {constB=0.8})
  n25 = Multiply(A=TextureCoordinate()) {constB=3}
  n26 = Multiply(
   A=VertexNormalWS()
   B=TextureSample(UVs=Panner(Coordinate=@n25) {speedX=0.5, speedY=0.5}) {samplerType=SAMPLERTYPE_LinearColor, texture=&HighContrastNoise}.RGB
  )
  n27 = DynamicParameter() {defaultValue=<rgba 1 1 1 50>, paramNames [Param1,Param2,Refraction,WPO]}
  Multiply(
   A=Saturate(None=Multiply(A=@n24) {constB=2})
   B=Divide(A=Multiply(A=@n26, B=@n27.WPO), B=ScalarParameter() {defaultValue=1, parameterName=WPO Off}) {constB=50}
  )
  n28 = Fresnel(
   ExponentIn=ScalarParameter() {defaultValue=1, parameterName=Fresnel_Exp}
   BaseReflectFractionIn=ScalarParameter() {defaultValue=0.1, parameterName=Fresnel_BRF}
  ) {baseReflectFraction=0.1, exponent=1}
  n29 = Multiply(A=@n18.A, B=Saturate(None=Power(Base=@n28)))
  n30 = TextureCoordinate() {vTiling=0.5}
  n31 = Multiply(A=Sine(None=ComponentMask(None=@n30) {g=true})) {constB=1.3}
  n32 = ScalarParameter() {defaultValue=8, parameterName=LinearMask_Power}
  n33 = ScalarParameter() {defaultValue=1, group=MaskTex, parameterName=MaskTexture_U}
  n34 = ScalarParameter() {defaultValue=1, group=MaskTex, parameterName=MaskTexture_V}
  n35 = TextureSampleParameter2D(UVs=Multiply(A=TextureCoordinate(), B=AppendVector(A=@n33, B=@n34))) {
   group=MaskTex
   parameterName=MaskTexture
   samplerType=SAMPLERTYPE_Grayscale
   texture=/Game/Effects/Fort_Effects/Textures/Smoke/T_Smoke_Trail_Soft_NoClamp
  }
  n36 = LinearInterpolate(
   A=Saturate(None=Power(Base=@n31, Exp=@n32) {constExponent=8})
   B=Power(Base=@n35.R, Exp=ScalarParameter() {defaultValue=1, group=MaskTex, parameterName=MaskTexture_power})
   Alpha=ScalarParameter() {group=MaskTex, parameterName=Use_MaskTexture}
  )
  n37 = Multiply(
   A=LinearInterpolate(A=Multiply(A=@n29, B=@n36), B=@n29, Alpha=ScalarParameter() {defaultValue=1, parameterName=Use_LinearMask})
   B=MaterialFunctionCall("FadeSensitivity (optional)"=ScalarParameter() {defaultValue=1.5, parameterName=MeshFade}) {materialFunction=&FadeBasedOnViewAngle}.Result
  )
  n38 = DepthFade(Opacity=@n37, FadeDistance=ScalarParameter() {defaultValue=5, parameterName=DepthFade})
  ShadingPathSwitch(Default=@n38, Deferred=@n38, Forward=@n37, Mobile=@n37)
  n39 = Multiply(A=Fresnel() {baseReflectFraction=0.1, exponent=1}, B=OneMinus(None=Fresnel() {baseReflectFraction=0.5, exponent=1}))
  n40 = Constant() {r=1}
  n41 = LinearInterpolate(
   A=LinearInterpolate(B=DynamicParameter() {defaultValue=<rgba 1 1 1 50>, paramNames [Param1,Param2,Refraction,WPO]}.Refraction, Alpha=Multiply(A=@n22, B=@n39)) {constA=1, constB=0}
   B=@n40
   Alpha=ScalarParameter() {defaultValue=1, parameterName=Use_Refraction}
  )
  ShadingPathSwitch(Default=@n41, Deferred=@n41, Forward=@n40, Mobile=@n40)
 ]
}
