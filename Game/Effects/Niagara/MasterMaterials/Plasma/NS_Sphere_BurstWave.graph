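# Niagara plasma burst-wave effect: a short-lived energy sphere that expands quickly and fades out.
# A single emitter spawns one mesh particle, with curves driving rapid-then-easing growth and a delayed alpha fade.
# A front-end float modifier shrinks the growth for menu contexts, and a linear ramp over lifetime animates a dynamic material parameter.
# User variables expose front-end flags and a custom color, though no visible module reads the color.
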NiagaraSystem /Game/Effects/Niagara/MasterMaterials/Plasma/NS_Sphere_BurstWave
dependencies {usedDataInterfaces [&NiagaraDataInterfaceCurve], usedDynamicInputs [], usedModules [&SystemState]}
emitters {Energy_Sphere {
  emitterUpdateScript [
   &EmitterState(Inactive Response=::NewEnumerator1, Life Cycle Mode=::NewEnumerator1, Loop Behavior=::NewEnumerator1, Loop Duration=1, UseLoopDelay=false)
   &SpawnBurst_Instantaneous(Spawn Count=1, Use Spawn Probability=true)
  ]
  particleSpawnScript [&InitializeParticle(Lifetime=0.5, Mesh Scale Mode=::NewEnumerator3, Mesh Uniform Scale=3, UsePositionOffset=true, Write Material Random=true) {module=InitializeParticle001}]
  particleUpdateScript [
   &ParticleState()
   &ScaleColor(
    Scale Alpha=&FloatFromCurve(FloatCurve=DataInterface() {propertyValues {Curve {keys [{arriveTangent=-2.4108, interpMode=RCIM_Cubic, leaveTangent=-2.4108, tangentMode=RCTM_User, time=0.3, value=1}, <at 1 0 RCIM_Cubic>]}}})
    ScaleA=true
    ScaleRGB=true
   )
   &ScaleMeshSize(Scale Factor=&VectorFromFloat(Value=&FloatFromCurve(
      FloatCurve=DataInterface() {propertyValues {Curve {keys [
          {arriveTangent=5.0067, interpMode=RCIM_Cubic, leaveTangent=5.0067, tangentMode=RCTM_User, time=0, value=0.15}
          {arriveTangent=0.6316, interpMode=RCIM_Cubic, leaveTangent=0.6316, tangentMode=RCTM_User, time=0.2048, value=0.9275}
          {arriveTangent=0.1704, interpMode=RCIM_Cubic, leaveTangent=0.1704, tangentMode=RCTM_User, time=1, value=1.2}
         ]}}}
      Scale Curve=&FN_FrontEndFloatModifier(Gameplay=1, Front End=0.78)
     )))
   &DynamicMaterialParameters(
    Index 0 Param 1=1
    Index 0 Param 2=&FloatFromCurve(FloatCurve=DataInterface() {propertyValues {Curve {keys [<at 0 1>,<at 1 0>]}}})
    Module.Index0Param0Write=true
    Module.Index0Param1Write=true
    Module.Index0Param2Write=true
    Module.Index0Param3Write=true
   )
  ]
  rendererClasses [&NiagaraMeshRendererProperties]
  renderers [{rendererClass=&NiagaraMeshRendererProperties, rendererIndex=0}]
 }}
summary {
 systemName=NS_Sphere_BurstWave
 emitters [{emitterName=Energy_Sphere, rendererClasses [&NiagaraMeshRendererProperties]}]
 userVariables [
  User.bIsFrontEnd {type=NiagaraBool}
  User.bIsFrontEndPreview {type=NiagaraBool}
  User.Color {type=LinearColor, description="", defaultValue {struct=LinearColor, value=<rgba 1 0 0.1618 1>}}
  User.UseCustomColor {type=NiagaraFloat, defaultValue=0, description=""}
 ]
}
